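# Translucent, unlit, two-sided Niagara mesh-particle material for the CeremonialGuard emote that wipes letter strokes on with a flame.
# Particle time drives a UV1 wipe, using Houdini-baked stroke length in vertex red to remap timing and equalize flame size.
# Flame and wiped stroke combine by max, erode through texture alpha, then scale HDR particle color for emissive.
# Dynamic parameters control wipe, erosion, intensity and a panning-texture warble intended for WPO.
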Material /BRCosmetics/Effects/Fort_Effects/Emotes/CeremonialGuard/M_Emote_CeremonialGuard
graph {
 properties {
  bCastRayTracedShadows=false
  bUsedWithNiagaraMeshParticles=true
  blendMode=BLEND_Translucent
  pixelDepthOffsetMode=PDOM_Legacy
  shadingModel=MSM_Unlit
  twoSided=true
  usedShadingModels=Unlit
 }
 nodes [
  n1 = TextureCoordinate() {coordinateIndex=1, desc=Flrame and Wipe Coordinates}
  n2 = MaterialFunctionCall(Float2=@n1) {materialFunction=&BreakOutFloat2Components}
  n3 = ScalarParameter() {defaultValue=1, parameterName=FlameWidthScalar}
  n4 = VertexColor() {desc="The R channel stores the rationalized \"length\" attribute of each stroke from Houdini."}
  n5 = Divide(B=OneMinus(None=@n4.R))
  n6 = Multiply(A=ParticleRelativeTime(), B=Constant() {r=0.5}) {constB=6}
  n7 = Constant()
  n8 = Constant()
  AppendVector(
   A=Add(A=Multiply(A=@n2.R, B=@n3), B=ScalarParameter() {parameterName=FlameWidthOffset})
   B=Subtract(
    A=Multiply(A=@n2.G, B=@n5)
    B=MaterialFunctionCall(Input=@n6, Input Low=@n7, Input High=@n4.R, Target Low=@n8, Target High=@n5) {desc=Remap from 0 and 1 to -1 and 1, materialFunction=&RemapValueRange}.Result
   )
  )
  DynamicParameter() {
   desc="Wipe: A 0-1 value that wipes the letter on. 0 = Off, 1 = On."
   defaultValue=<rgba 0.9 1 1 1>
   paramNames [Wipe Speed, Alpha Erode, Intensity, WPO Texture Scale]
  }
  DynamicParameter() {
   desc=Speed multiplier on Particle Relative Time
   defaultValue=<rgba 0.9 1 1 1>
   paramNames [Wipe Speed, Alpha Erode, Intensity, WPO Texture Scale]
  }
  MaterialFunctionCall(XY=@n1) {materialFunction=&Swizzle}
  n9 = ScalarParameter() {
   defaultValue=0.1
   desc=A value of 1 makes the flame be the length of the entire strip. Smaller values shorten the length.
   parameterName=FlameLengthScalar
  }
  MaterialFunctionCall(
   Input=Multiply(A=ParticleRelativeTime(), B=Constant() {r=0.5}) {constB=6}
   Target Low=Multiply(A=@n9, B=Constant() {r=1})
   Target High=Multiply(A=@n9, B=Constant() {r=-1})
  ) {materialFunction=&RemapValueRange}
  n10 = ScalarParameter() {
   defaultValue=0.1
   desc=A value of 1 makes the flame be the length of the entire strip. Smaller values shorten the length.
   parameterName=FlameLengthScalar
  }
  Multiply(A=@n10, B=@n4.R) {desc=Scale By stroke length}
  n11 = DynamicParameter() {
   desc=Speed multiplier on Particle Relative Time
   defaultValue=<rgba 0.9 1 1 1>
   paramNames [Wipe Speed, Alpha Erode, Intensity, WPO Texture Scale]
  }
  n12 = Panner(
   Coordinate=Multiply(A=TextureCoordinate(), B=@n11.WPO Texture Scale)
   Speed=ScalarParameter() {defaultValue=0.05, parameterName=Warble Speed}
  ) {bFractionalPart=true, speedX=0.2, speedY=0.2}
  n13 = ComponentMask(None=TextureCoordinate()) {g=true}
  Multiply(
   A=Multiply(
    A=TextureSample(UVs=@n12) {samplerSource=SSM_Wrap_WorldGroupSettings, texture=/Game/Environments/Sets/Junkyard/Textures/T_Flag_M}.G
    B=DynamicParameter() {parameterIndex=1, defaultValue=<rgba 0.9 1 1 1>, paramNames [WPO Height, Param2, Param3, Param4]}.WPO Height
   ) {constB=10}
   B=Clamp(None=Power(Base=@n13) {constExponent=0.5}) {minDefault=0.2}
  )
  Multiply(
   A=Add(B=ScalarParameter() {defaultValue=0.2, desc=Offset and tweak the flame offset to match the wipe edge., parameterName=FlameLengthOffset})
   B=@n9
  )
  n14 = MaterialFunctionCall(Float2=TextureCoordinate() {coordinateIndex=1}) {materialFunction=&BreakOutFloat2Components}
  NamedRerouteDeclaration(None=@n14.G) {name=WipeVCoord, nodeColor=<rgba 1 0.1875 0 1>}
  n15 = OneMinus(None=@n14.R)
  n16 = DynamicParameter() {
   desc=Speed multiplier on Particle Relative Time
   defaultValue=<rgba 0.9 1 1 1>
   paramNames [Wipe Speed, Alpha Erode, Intensity, WPO Texture Scale]
  }
  n17 = MaterialFunctionCall(
   Input=Multiply(A=ParticleRelativeTime(), B=@n16.Wipe Speed) {constB=6}
   Input Low=Constant()
   Input High=VertexColor() {desc="The R channel stores the rationalized \"length\" attribute of each stroke from Houdini."}.R
   Target Low=Constant() {r=-1}
   Target High=Constant() {r=1}
  ) {desc=Remap from 0 and 1 to -1 and 1, materialFunction=&RemapValueRange}
  NamedRerouteDeclaration(None=Saturate(None=Add(A=@n15, B=@n17.Result))) {name=WipeUCoord, nodeColor=<rgba 0.3125 0 1 1>}
  NamedRerouteDeclaration(None=Transform(None=VectorParameter() {parameterName=WPO_Offset}.RGB) {transformSourceType=TRANSFORMSOURCE_Camera}) {name=WPO, nodeColor=<rgba 1 0.0938 0 1>}
  n18 = ParticleColor()
  n19 = DynamicParameter() {
   desc=Speed multiplier on Particle Relative Time
   defaultValue=<rgba 0.9 1 1 1>
   paramNames [Wipe Speed, Alpha Erode, Intensity, WPO Texture Scale]
  }
  n20 = MaterialFunctionCall(LDR Input=Clamp(None=@n18.RGB), HDR Tint=@n18.RGB, HDR Intensity Multiplier=@n19.Intensity) {materialFunction=&DeriveHDRfromLDR}
  n21 = TextureCoordinate() {coordinateIndex=1}
  n22 = OneMinus(None=Cosine(None=MaterialFunctionCall(Float2=@n21) {materialFunction=&BreakOutFloat2Components}.R))
  n23 = ScalarParameter() {
   defaultValue=0.1
   desc=A value of 1 makes the flame be the length of the entire strip. Smaller values shorten the length.
   parameterName=FlameLengthScalar
  }
  n24 = Multiply(
   A=Multiply(A=NamedRerouteUsage(), B=Divide(B=@n23)) {constB=10}
   B=VertexColor() {desc="The R channel stores the rationalized \"length\" attribute of each stroke from Houdini."}.R
  ) {desc=Multiply by relative lengths to equalize the size of fireballs per stroke.}
  n25 = NamedRerouteUsage()
  n26 = MaterialFunctionCall(
   In=TextureSample(
    UVs=AppendVector(A=@n24, B=@n25)
    Tex=TextureObjectParameter() {parameterName=FlameTexture, texture=&/T_Emote_CeremonialGuard_Flame}
   ) {samplerSource=SSM_Clamp_WorldGroupSettings}.R
   Contrast=ScalarParameter() {defaultValue=2, parameterName="Contrast (S)"}
  ) {materialFunction=&CheapContrast}
  n27 = TextureSample(UVs=TextureCoordinate() {desc=Stroke Coordinates}, Tex=TextureObjectParameter() {parameterName=StrokeTexture})
  n28 = NamedRerouteUsage()
  n29 = ScalarParameter() {defaultValue=1, parameterName=Wipe Exponent}
  n30 = Max(
   A=Multiply(A=Power(Base=@n22) {constExponent=3}, B=Saturate(None=@n26.Result))
   B=Multiply(A=ChannelMaskParameter(None=@n27.RGB) {parameterName=Mask Stroke}.R, B=Power(Base=@n28, Exp=@n29)) {desc=Mask out stroke with wipe.}
  ) {desc=Combine Flame and Wipe}
  n31 = Power(Base=@n27.A, Exp=@n11.Alpha Erode)
  n32 = Saturate(None=Multiply(A=@n18.A, B=Multiply(A=@n30, B=@n31) {desc=Erode}))
  Reroute(None=MaterialFunctionCall(Emissive=Multiply(A=@n20.RemappedRange, B=@n32)) {materialFunction=&MF_FX_EmissiveAdjustments}.Result) {desc=EMISSIVE}
 ]
}
