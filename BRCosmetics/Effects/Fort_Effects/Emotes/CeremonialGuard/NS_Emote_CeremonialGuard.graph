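# Niagara system for the Ceremonial Guard emote that draws a one- or two-digit number, apparently the kill count from User.NumKills, in animated strokes.
# System variables feed TensPlace and OnesPlace mesh emitters, which burst up to three timed strokes per digit via lookup arrays and visibility tags.
# IsFacingPlayer flips kerning and mesh facing, while material parameters animate the strokes before fading.
# Older-style WPO-offset duplicates and a camera-facing debug emitter are disabled.
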NiagaraSystem /BRCosmetics/Effects/Fort_Effects/Emotes/CeremonialGuard/NS_Emote_CeremonialGuard
dependencies {
 usedDataInterfaces [&NiagaraDataInterfaceArrayInt32, &NiagaraDataInterfaceCurve, &NiagaraDataInterfaceCamera]
 usedDynamicInputs [
  &MakeIntFromFloat
  &Divide_Float
  /Niagara/DynamicInputs/Subtract/Subtract_Integer
  &Multiply_Int
  &SetBoolByIntComparison
  &MakeCustomIntFromBool
 ]
 usedModules [
  &self:SystemSpawnScript.SystemScriptSource.SystemScriptGraph.NiagaraNodeAssignment_0.SetVariables_#1
  &self:SystemSpawnScript.SystemScriptSource.SystemScriptGraph.NiagaraNodeAssignment_1.SetVariables_#2
  &self:SystemSpawnScript.SystemScriptSource.SystemScriptGraph.NiagaraNodeAssignment_3.SetVariables_#3
  &self:SystemSpawnScript.SystemScriptSource.SystemScriptGraph.NiagaraNodeAssignment_2.SetVariables_#4
  &self:SystemSpawnScript.SystemScriptSource.SystemScriptGraph.NiagaraNodeAssignment_4.SetVariables_#5
  &self:SystemSpawnScript.SystemScriptSource.SystemScriptGraph.NiagaraNodeAssignment_5.SetVariables_#6
  &self:SystemSpawnScript.SystemScriptSource.SystemScriptGraph.NiagaraNodeAssignment_6.SetVariables_#7
  &self:SystemSpawnScript.SystemScriptSource.SystemScriptGraph.NiagaraNodeAssignment_7.SetVariables_#8
  &SystemState
  &self:SystemSpawnScript.SystemScriptSource.SystemScriptGraph.NiagaraNodeAssignment_9.SetVariables_#9
 ]
}
emitters {
 FN_Debug {
  bEnabled=false
  emitterUpdateScript [
   &EmitterState()
   &SpawnBurst_Instantaneous(Spawn Count=1, Use Spawn Probability=false)
  ]
  particleSpawnScript [
   /Niagara/Modules/Spawn/Initialization/InitializeParticle(
    Lifetime=5
    Module.Write Color=true
    Module.Write Lifetime=true
    Module.Write Mass=true
    Module.Write Material Random=true
    Module.Write Position=false
    Module.Write Scale=false
    Module.Write SpriteRotation=false
    Module.Write SpriteSize=true
    Position=&SimulationPosition()
    Sprite Size=<100 100>
   ) {enabled=false}
   /Niagara/Modules/Spawn/Initialization/V2/InitializeParticle(Lifetime=5, Position Offset=<-50 80 150>, UsePositionOffset=true, Write Material Random=true) {module=InitializeParticle001}
  ]
  particleUpdateScript [
   &ParticleState()
   /SetVariables(Particles.Debug:NiagaraFloat=&MakeCustomFloatFromBool(
     Bool=&SetBoolByFloatComparison(
      A=&DotProduct(
       Vector A=$Engine.Owner.SystemYAxis
       Vector B=&ReturnCameraVector(Camera Vectors=::NewEnumerator1, Coordinate Space=::World, CameraQuery=DataInterface() {propertyValues {bRequireCurrentFrameData=false}})
      )
      B=0
      Comparison Type=::NewEnumerator2
     )
     True Float=1
     False Float=0
    )) {isSetParameters=true}
   &DynamicMaterialParameters(
    Index 0 Param 1=$Particles.Debug
    Module.Param0WriteEnabled=true
    Module.Param1WriteEnabled=true
    Module.Param2WriteEnabled=true
    Module.Param3WriteEnabled=true
   )
  ]
  rendererClasses [&NiagaraSpriteRendererProperties]
  renderers [{rendererClass=&NiagaraSpriteRendererProperties}]
 }
 OnesPlace {
  emitterSpawnScript [/SetVariables(Emitter.MeshIndex:NiagaraInt32=$System.OnesPlace) {isSetParameters=true, module=SetVariables_#10}]
  emitterUpdateScript [
   &EmitterState(Life Cycle Mode=::NewEnumerator1, Loop Behavior=::NewEnumerator1, Loop Duration Mode=::NewEnumerator1, UseLoopDelay=false)
   &SpawnBurst_Instantaneous(Spawn Time=&Add_Float(A=0, B=$System.OnesPlaceDelay), Use Spawn Probability=false, UseLoopCountLimit=false)
   &SpawnBurst_Instantaneous(
    Spawn Count=&SelectIntFromArray(Direct Array Index=$Emitter.MeshIndex, Int Selection Array=DataInterface() {propertyValues {IntData [0,1,1,1,1,1,0,1,0,0]}})
    Spawn Group=1
    Spawn Time=&Add_Float(A=0.17, B=$System.OnesPlaceDelay)
    Use Spawn Probability=false
    UseLoopCountLimit=false
   ) {module=SpawnBurst_Instantaneous001}
   &SpawnBurst_Instantaneous(
    Spawn Count=&SelectIntFromArray(Direct Array Index=$Emitter.MeshIndex, Int Selection Array=DataInterface() {propertyValues {IntData [0,0,0,0,1,1,0,1,0,0]}})
    Spawn Group=2
    Spawn Time=&Add_Float(A=0.33, B=$System.OnesPlaceDelay)
    Use Spawn Probability=false
    UseLoopCountLimit=false
   ) {module=SpawnBurst_Instantaneous002}
  ]
  particleSpawnScript [
   /SetVariables(Particles.VisibilityTag=&SpawnGroup()) {isSetParameters=true, module=SetVariables_#11}
   &InitializeParticle(
    Color=$System.FontColor
    Lifetime=$System.StrokeLife
    Mesh Scale Mode=::NewEnumerator3
    Mesh Uniform Scale=0.7
    UsePositionOffset=false
    Write Material Random=true
   )
   /SetVariables(Particles.MeshIndex=$Emitter.MeshIndex) {isSetParameters=true, module=SetVariables_#12}
  ]
  particleUpdateScript [
   &ParticleState(Kill Particles When Lifetime Has Elapsed=false, Let Infinitely Lived Particles Die When Emitter Deactivates=true)
   /SetVariables(Particles.UpdateMeshPosition:NiagaraPosition=&AddVectorToPosition(
     Position=$Particles.Position
     Vector=&MakeVector(
      X=&Multiply_Float(
       A=$System.Kerning
       B=&MakeCustomFloatFromBool(Bool=$System.IsFacingPlayer, True Float=-1, False Float=1)
      )
      Y=0
      Z=0
     )
    )) {isSetParameters=true}
   &UpdateMeshOrientation(
    Facing Direction=&MakeCustomVectorFromBool(Bool=$System.IsFacingPlayer, True Vector=<0 -1 0>, False Vector=<0 1 0>)
    Facing Interpolation=false
    Orientation Method=::NewEnumerator1
    Override Quat=false
    Stabilization Mode:ENiagara_OrientVectorCount=::NewEnumerator0
    TransformFacing=false
   )
   &DynamicMaterialParameters(
    Index 0 Param 1=$System.Wipe Speed
    Index 0 Param 2=&FloatFromCurve(
     FloatCurve=DataInterface() {propertyValues {Curve {keys [
         {interpMode=RCIM_Cubic, leaveTangent=978.0479, tangentMode=RCTM_Break, time=2.5, value=0}
         {arriveTangent=46.4813, tangentMode=RCTM_User, time=3.5, value=500}
        ]}}}
     CurveIndex=$System.Age
    )
    Index 0 Param 3=&FloatFromCurve(
     FloatCurve=DataInterface() {propertyValues {Curve {keys [
         {interpMode=RCIM_Cubic, tangentMode=RCTM_User, time=0, value=0}
         {interpMode=RCIM_Cubic, tangentMode=RCTM_User, time=0.05, value=0.85}
         {arriveTangent=-3.9369, interpMode=RCIM_Cubic, leaveTangent=-3.9369, tangentMode=RCTM_User, time=0.15, value=0.45}
         {interpMode=RCIM_Cubic, tangentMode=RCTM_User, time=0.3, value=0.1}
        ]}}}
     Scale Curve=25
    )
    Index 0 Param 4=0.1
    Index 1 Param 1=&FloatFromCurve(FloatCurve=DataInterface() {propertyValues {Curve {keys [<at 0.25 0>,<at 1 1>]}}}, CurveIndex=$Particles.Age, Scale Curve=20)
    Module.Index0Param0Write=true
    Module.Index0Param1Write=true
    Module.Index0Param2Write=true
    Module.Index0Param3Write=true
    Module.Index1Param0Write=true
    Module.Index1Param1Write=true
    Module.Index1Param2Write=true
    Module.Index1Param3Write=true
    Write Parameter Index 1=true
   )
   &ScaleColor(
    Scale Alpha=&FloatFromCurve(
     FloatCurve=DataInterface() {propertyValues {Curve {keys [<at 2.5 1 RCIM_Cubic>, {arriveTangent=-5.5723, tangentMode=RCTM_User, time=2.75, value=0}]}}}
     CurveIndex=$System.Age
    )
    ScaleA=true
    ScaleRGB=false
   )
  ]
  rendererClasses [&NiagaraMeshRendererProperties]
  renderers [
   {rendererClass=&NiagaraMeshRendererProperties, rendererIndex=0}
   {rendererClass=&NiagaraMeshRendererProperties, rendererIndex=1}
   {rendererClass=&NiagaraMeshRendererProperties, rendererIndex=2}
  ]
 }
 OnesPlace001 {
  bEnabled=false
  emitterSpawnScript [/SetVariables(Emitter.MeshIndex:NiagaraInt32=$System.OnesPlace) {isSetParameters=true, module=SetVariables_#10}]
  emitterUpdateScript [
   &EmitterState(Life Cycle Mode=::NewEnumerator1, Loop Behavior=::NewEnumerator1, Loop Duration Mode=::NewEnumerator1, UseLoopDelay=false)
   &SpawnBurst_Instantaneous(Spawn Time=&Add_Float(A=0, B=$System.OnesPlaceDelay), Use Spawn Probability=false, UseLoopCountLimit=false)
   &SpawnBurst_Instantaneous(
    Spawn Count=&SelectIntFromArray(Direct Array Index=$Emitter.MeshIndex, Int Selection Array=DataInterface() {propertyValues {IntData [0,1,1,1,1,1,0,1,0,0]}})
    Spawn Group=1
    Spawn Time=&Add_Float(A=0.17, B=$System.OnesPlaceDelay)
    Use Spawn Probability=false
    UseLoopCountLimit=false
   ) {module=SpawnBurst_Instantaneous001}
   &SpawnBurst_Instantaneous(
    Spawn Count=&SelectIntFromArray(Direct Array Index=$Emitter.MeshIndex, Int Selection Array=DataInterface() {propertyValues {IntData [0,0,0,0,1,1,0,1,0,0]}})
    Spawn Group=2
    Spawn Time=&Add_Float(A=0.33, B=$System.OnesPlaceDelay)
    Use Spawn Probability=false
    UseLoopCountLimit=false
   ) {module=SpawnBurst_Instantaneous002}
   /SetVariables(Emitter.WPO_Offset:Vector3f=&MakeVector(X=&Multiply_Float(A=$System.Kerning, B=1), Y=0, Z=0)) {isSetParameters=true}
  ]
  particleSpawnScript [
   /SetVariables(Particles.VisibilityTag=&SpawnGroup()) {isSetParameters=true, module=SetVariables_#11}
   &InitializeParticle(
    Color=$System.FontColor
    Lifetime=$System.StrokeLife
    Mesh Scale Mode=::NewEnumerator3
    Mesh Uniform Scale=0.7
    UsePositionOffset=true
    Write Material Random=true
   )
   /SetVariables(Particles.MeshIndex=$Emitter.MeshIndex) {isSetParameters=true, module=SetVariables_#12}
  ]
  particleUpdateScript [
   &ParticleState(Kill Particles When Lifetime Has Elapsed=false, Let Infinitely Lived Particles Die When Emitter Deactivates=true)
   &DynamicMaterialParameters(
    Index 0 Param 1=$System.Wipe Speed
    Index 0 Param 2=&FloatFromCurve(
     FloatCurve=DataInterface() {propertyValues {Curve {keys [
         {interpMode=RCIM_Cubic, leaveTangent=978.0479, tangentMode=RCTM_Break, time=2, value=0}
         {arriveTangent=46.4813, tangentMode=RCTM_User, time=3, value=500}
        ]}}}
     CurveIndex=$System.Age
    )
    Index 0 Param 3=&FloatFromCurve(
     FloatCurve=DataInterface() {propertyValues {Curve {keys [
         {interpMode=RCIM_Cubic, tangentMode=RCTM_User, time=0, value=0}
         {interpMode=RCIM_Cubic, tangentMode=RCTM_User, time=0.05, value=0.85}
         {arriveTangent=-3.9369, interpMode=RCIM_Cubic, leaveTangent=-3.9369, tangentMode=RCTM_User, time=0.15, value=0.45}
         {interpMode=RCIM_Cubic, tangentMode=RCTM_User, time=0.3, value=0.1}
        ]}}}
     Scale Curve=25
    )
    Index 0 Param 4=0.1
    Index 1 Param 1=&FloatFromCurve(
     FloatCurve=DataInterface() {propertyValues {Curve {keys [{tangentMode=RCTM_User, time=0, value=0}, {tangentMode=RCTM_User, time=1, value=1}]}}}
     CurveIndex=$Particles.Age
     Scale Curve=2
    )
    Module.Index0Param0Write=true
    Module.Index0Param1Write=true
    Module.Index0Param2Write=true
    Module.Index0Param3Write=true
    Module.Index1Param0Write=true
    Module.Index1Param1Write=true
    Module.Index1Param2Write=true
    Module.Index1Param3Write=true
    Write Parameter Index 1=true
   )
   &ScaleColor(
    Scale Alpha=&FloatFromCurve(
     FloatCurve=DataInterface() {propertyValues {Curve {keys [<at 2 1 RCIM_Cubic>, {arriveTangent=-5.5723, tangentMode=RCTM_User, time=2.25, value=0}]}}}
     CurveIndex=$System.Age
    )
    ScaleA=true
    ScaleRGB=false
   )
  ]
  rendererClasses [&NiagaraMeshRendererProperties]
  renderers [
   {rendererClass=&NiagaraMeshRendererProperties, rendererIndex=0}
   {rendererClass=&NiagaraMeshRendererProperties, rendererIndex=1}
   {rendererClass=&NiagaraMeshRendererProperties, rendererIndex=2}
  ]
 }
 TensPlace {
  emitterSpawnScript [/SetVariables(Emitter.MeshIndex:NiagaraInt32=$System.TensPlace) {isSetParameters=true, module=SetVariables_#10}]
  emitterUpdateScript [
   &EmitterState(Life Cycle Mode=::NewEnumerator1, Loop Behavior=::NewEnumerator1, Loop Duration Mode=::NewEnumerator1, UseLoopDelay=false)
   &SpawnBurst_Instantaneous(
    Spawn Probability=&MakeCustomFloatFromBool(Bool=$System.IsDoubleDigit, True Float=1, False Float=0)
    Use Spawn Probability=true
    UseLoopCountLimit=false
   )
   &SpawnBurst_Instantaneous(
    Spawn Count=&SelectIntFromArray(Direct Array Index=$Emitter.MeshIndex, Int Selection Array=DataInterface() {propertyValues {IntData [0,1,1,1,1,1,0,1,0,0]}})
    Spawn Group=1
    Spawn Probability=&MakeCustomFloatFromBool(Bool=$System.IsDoubleDigit, True Float=1, False Float=0)
    Spawn Time=0.17
    Use Spawn Probability=true
    UseLoopCountLimit=false
   ) {module=SpawnBurst_Instantaneous001}
   &SpawnBurst_Instantaneous(
    Spawn Count=&SelectIntFromArray(Direct Array Index=$Emitter.MeshIndex, Int Selection Array=DataInterface() {propertyValues {IntData [0,0,0,0,1,1,0,1,0,0]}})
    Spawn Group=2
    Spawn Probability=&MakeCustomFloatFromBool(Bool=$System.IsDoubleDigit, True Float=1, False Float=0)
    Spawn Time=0.33
    Use Spawn Probability=true
    UseLoopCountLimit=false
   ) {module=SpawnBurst_Instantaneous002}
  ]
  particleSpawnScript [
   /SetVariables(Particles.VisibilityTag=&SpawnGroup()) {isSetParameters=true, module=SetVariables_#11}
   &InitializeParticle(
    Color=$System.FontColor
    Lifetime=$System.StrokeLife
    Mesh Scale Mode=::NewEnumerator3
    Mesh Uniform Scale=0.7
    UsePositionOffset=false
    Write Material Random=true
   )
   /SetVariables(Particles.MeshIndex=$Emitter.MeshIndex) {isSetParameters=true, module=SetVariables_#12}
  ]
  particleUpdateScript [
   &ParticleState(Kill Particles When Lifetime Has Elapsed=false, Let Infinitely Lived Particles Die When Emitter Deactivates=true)
   /SetVariables(Particles.UpdateMeshPosition:NiagaraPosition=&AddVectorToPosition(
     Position=$Particles.Position
     Vector=&MakeVector(
      X=&Multiply_Float(
       A=$System.Kerning
       B=&MakeCustomFloatFromBool(Bool=$System.IsFacingPlayer, True Float=1, False Float=-1)
      )
      Y=0
      Z=0
     )
    )) {isSetParameters=true}
   &UpdateMeshOrientation(
    Facing Direction=&MakeCustomVectorFromBool(Bool=$System.IsFacingPlayer, True Vector=<0 -1 0>, False Vector=<0 1 0>)
    Facing Interpolation=false
    Orientation Method=::NewEnumerator1
    Override Quat=false
    Stabilization Mode:ENiagara_OrientVectorCount=::NewEnumerator0
    TransformFacing=false
   )
   &DynamicMaterialParameters(
    Index 0 Param 1=$System.Wipe Speed
    Index 0 Param 2=&FloatFromCurve(
     FloatCurve=DataInterface() {propertyValues {Curve {keys [
         {interpMode=RCIM_Cubic, leaveTangent=978.0479, tangentMode=RCTM_Break, time=2.5, value=0}
         {arriveTangent=46.4813, tangentMode=RCTM_User, time=3.5, value=500}
        ]}}}
     CurveIndex=$System.Age
    )
    Index 0 Param 3=&FloatFromCurve(
     FloatCurve=DataInterface() {propertyValues {Curve {keys [
         {interpMode=RCIM_Cubic, tangentMode=RCTM_User, time=0, value=0}
         {interpMode=RCIM_Cubic, tangentMode=RCTM_User, time=0.05, value=0.85}
         {arriveTangent=-3.9369, interpMode=RCIM_Cubic, leaveTangent=-3.9369, tangentMode=RCTM_User, time=0.15, value=0.45}
         {interpMode=RCIM_Cubic, tangentMode=RCTM_User, time=0.3, value=0.1}
        ]}}}
     Scale Curve=25
    )
    Index 0 Param 4=0.1
    Index 1 Param 1=&FloatFromCurve(FloatCurve=DataInterface() {propertyValues {Curve {keys [<at 0 0.5>,<at 1 1>]}}}, CurveIndex=$Particles.Age, Scale Curve=20)
    Module.Index0Param0Write=true
    Module.Index0Param1Write=true
    Module.Index0Param2Write=true
    Module.Index0Param3Write=true
    Module.Index1Param0Write=true
    Module.Index1Param1Write=true
    Module.Index1Param2Write=true
    Module.Index1Param3Write=true
    Write Parameter Index 1=true
   )
   &ScaleColor(
    Scale Alpha=&FloatFromCurve(
     FloatCurve=DataInterface() {propertyValues {Curve {keys [<at 2.5 1 RCIM_Cubic>, {arriveTangent=-5.5723, tangentMode=RCTM_User, time=2.75, value=0}]}}}
     CurveIndex=$System.Age
    )
    ScaleA=true
    ScaleRGB=false
   )
  ]
  rendererClasses [&NiagaraMeshRendererProperties]
  renderers [
   {rendererClass=&NiagaraMeshRendererProperties, rendererIndex=0}
   {rendererClass=&NiagaraMeshRendererProperties, rendererIndex=1}
   {rendererClass=&NiagaraMeshRendererProperties, rendererIndex=2}
  ]
 }
 TensPlace001 {
  bEnabled=false
  emitterSpawnScript [/SetVariables(Emitter.MeshIndex:NiagaraInt32=$System.TensPlace) {isSetParameters=true, module=SetVariables_#10}]
  emitterUpdateScript [
   &EmitterState(Life Cycle Mode=::NewEnumerator1, Loop Behavior=::NewEnumerator1, Loop Duration Mode=::NewEnumerator1, UseLoopDelay=false)
   &SpawnBurst_Instantaneous(
    Spawn Probability=&MakeCustomFloatFromBool(Bool=$System.IsDoubleDigit, True Float=1, False Float=0)
    Use Spawn Probability=true
    UseLoopCountLimit=false
   )
   &SpawnBurst_Instantaneous(
    Spawn Count=&SelectIntFromArray(Direct Array Index=$Emitter.MeshIndex, Int Selection Array=DataInterface() {propertyValues {IntData [0,1,1,1,1,1,0,1,0,0]}})
    Spawn Group=1
    Spawn Probability=&MakeCustomFloatFromBool(Bool=$System.IsDoubleDigit, True Float=1, False Float=0)
    Spawn Time=0.17
    Use Spawn Probability=true
    UseLoopCountLimit=false
   ) {module=SpawnBurst_Instantaneous001}
   &SpawnBurst_Instantaneous(
    Spawn Count=&SelectIntFromArray(Direct Array Index=$Emitter.MeshIndex, Int Selection Array=DataInterface() {propertyValues {IntData [0,0,0,0,1,1,0,1,0,0]}})
    Spawn Group=2
    Spawn Probability=&MakeCustomFloatFromBool(Bool=$System.IsDoubleDigit, True Float=1, False Float=0)
    Spawn Time=0.33
    Use Spawn Probability=true
    UseLoopCountLimit=false
   ) {module=SpawnBurst_Instantaneous002}
   /SetVariables(Emitter.WPO_Offset:Vector3f=&MakeVector(X=&Multiply_Float(A=$System.Kerning, B=-1), Y=0, Z=0)) {isSetParameters=true}
  ]
  particleSpawnScript [
   /SetVariables(Particles.VisibilityTag=&SpawnGroup()) {isSetParameters=true, module=SetVariables_#11}
   &InitializeParticle(
    Color=$System.FontColor
    Lifetime=$System.StrokeLife
    Mesh Scale Mode=::NewEnumerator3
    Mesh Uniform Scale=0.7
    UsePositionOffset=true
    Write Material Random=true
   )
   /SetVariables(Particles.MeshIndex=$Emitter.MeshIndex) {isSetParameters=true, module=SetVariables_#12}
  ]
  particleUpdateScript [
   &ParticleState(Kill Particles When Lifetime Has Elapsed=false, Let Infinitely Lived Particles Die When Emitter Deactivates=true)
   &DynamicMaterialParameters(
    Index 0 Param 1=$System.Wipe Speed
    Index 0 Param 2=&FloatFromCurve(
     FloatCurve=DataInterface() {propertyValues {Curve {keys [
         {interpMode=RCIM_Cubic, leaveTangent=978.0479, tangentMode=RCTM_Break, time=2, value=0}
         {arriveTangent=46.4813, tangentMode=RCTM_User, time=3, value=500}
        ]}}}
     CurveIndex=$System.Age
    )
    Index 0 Param 3=&FloatFromCurve(
     FloatCurve=DataInterface() {propertyValues {Curve {keys [
         {interpMode=RCIM_Cubic, tangentMode=RCTM_User, time=0, value=0}
         {interpMode=RCIM_Cubic, tangentMode=RCTM_User, time=0.05, value=0.85}
         {arriveTangent=-3.9369, interpMode=RCIM_Cubic, leaveTangent=-3.9369, tangentMode=RCTM_User, time=0.15, value=0.45}
         {interpMode=RCIM_Cubic, tangentMode=RCTM_User, time=0.3, value=0.1}
        ]}}}
     Scale Curve=25
    )
    Index 0 Param 4=1
    Index 1 Param 1=&FloatFromCurve(
     FloatCurve=DataInterface() {propertyValues {Curve {keys [<at 0.25 0 RCIM_Cubic>, {arriveTangent=2.0599, tangentMode=RCTM_User, time=1, value=1}]}}}
     CurveIndex=$System.Age
     Scale Curve=2
    )
    Module.Index0Param0Write=true
    Module.Index0Param1Write=true
    Module.Index0Param2Write=true
    Module.Index0Param3Write=true
    Module.Index1Param0Write=true
    Module.Index1Param1Write=true
    Module.Index1Param2Write=true
    Module.Index1Param3Write=true
    Write Parameter Index 1=true
   )
   &ScaleColor(
    Scale Alpha=&FloatFromCurve(
     FloatCurve=DataInterface() {propertyValues {Curve {keys [<at 2 1 RCIM_Cubic>, {arriveTangent=-5.5723, tangentMode=RCTM_User, time=2.25, value=0}]}}}
     CurveIndex=$System.Age
    )
    ScaleA=true
    ScaleRGB=false
   )
  ]
  rendererClasses [&NiagaraMeshRendererProperties]
  renderers [
   {rendererClass=&NiagaraMeshRendererProperties, rendererIndex=0}
   {rendererClass=&NiagaraMeshRendererProperties, rendererIndex=1}
   {rendererClass=&NiagaraMeshRendererProperties, rendererIndex=2}
  ]
 }
}
summary {
 systemName=NS_Emote_CeremonialGuard
 emitters [
  {emitterName=OnesPlace, rendererClasses [&NiagaraMeshRendererProperties]}
  {emitterName=TensPlace, rendererClasses [&NiagaraMeshRendererProperties]}
  {bEnabled=false, emitterName=OnesPlace001, rendererClasses [&NiagaraMeshRendererProperties]}
  {bEnabled=false, emitterName=TensPlace001, rendererClasses [&NiagaraMeshRendererProperties]}
  {bEnabled=false, emitterName=FN_Debug, rendererClasses [&NiagaraSpriteRendererProperties]}
 ]
 userVariables [User.NumKills {type=NiagaraFloat, defaultValue=32, description=""}]
}
